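# Niagara explosion burst for the WavySignalBeef player emote, built from flashes, sparks and mist.
# Three FlashElements emitters burst short-lived star and flare sprites with fading blue, orange and hot-to-red lights, with front-end overrides for size.
# Sparks fly outward and up under drag and delayed gravity, while snow_mist puffs from a torus with time-of-day tint.
# Particle aging, plus spark and mist motion, follows SystemTimeScale.
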NiagaraSystem /BRCosmetics/Animation/Game/MainPlayer/Emotes/WavySignalBeef/FX/NS_Emote_WavySignalBeef_Explosion_New
dependencies {
 usedDataInterfaces [&NiagaraDataInterfaceColorCurve, &NiagaraDataInterfaceCurve, &NiagaraDataInterfaceVectorCurve]
 usedDynamicInputs [&Multiply_VectorByFloat]
 usedModules [
  &SystemState
  &self:SystemSpawnScript.SystemScriptSource.SystemScriptGraph.NiagaraNodeAssignment_0.SetVariables_#1
 ]
}
emitters {
 FlashElements {
  emitterUpdateScript [
   &EmitterState(Life Cycle Mode=::NewEnumerator1, Loop Behavior=::NewEnumerator1, Loop Duration=0.06, UseLoopDelay=true)
   &SpawnBurst_Instantaneous(Spawn Group=1, Use Spawn Probability=false, UseLoopCountLimit=false) {module=SpawnBurst_Instantaneous001}
   &SpawnBurst_Instantaneous(
    Spawn Count=&RandomRangeInt(Minimum=6, Maximum=8)
    Use Spawn Probability=false
    UseLoopCountLimit=false
   )
  ]
  particleSpawnScript [
   &InitializeParticle(
    Color Mode=::NewEnumerator0
    Position Offset=<0 0 30>
    Sprite Rotation Mode=::NewEnumerator0
    UsePositionOffset=true
    Write Material Random=true
   )
   &ShapeLocation(MaskSpawnGroup=false, Override Seed=false, Sphere Radius=1, UseOwnerScale=true)
   &SpriteFacingAndAlignment(Sprite Alignment=$Particles.ShapeLocation.ShapeNormal, WriteAlignment=true, WriteFacing=false)
   &FN_PerParticleLOD(Enable SpriteSize Distance Scaling=true, Sprite Size Scale Far=<2.5 2.5>)
   /SetVariables(Particles.SpriteSize=<1 1>, Particles.VisibilityTag=&SpawnGroup()) {isSetParameters=true, module=SetVariables_#2}
   /SetVariables(
    Particles.Flash Star Cam Offset=15
    Particles.Flash Star Lifetime=0.15
    Particles.Flash Star Rotation:NiagaraFloat=&RandomRangeFloat(Minimum=0, Maximum=360)
    Particles.Flash Star Sprite Size:Vector2f=&Multiply_Vector2D_ByFloat(
     Float=&Multiply_Float(
      A=&MakeCustomFloatFromBool(Bool=$User.bIsFrontEnd, True Float=$User.FrontEndScale, False Float=1.5)
      B=&RandomRangeFloat(Minimum=110, Maximum=130)
     )
     Vector2D=$Particles.SpriteSize
    )
    Particles.Flash Star SubUV Index:NiagaraFloat=&RandomRangeFloat(Minimum=0, Maximum=3)
   ) {isSetParameters=true, module=SetVariables_#3}
   /SetVariables(
    Particles.Flash Flare Lifetime:NiagaraFloat=&Multiply_Float(A=0.4, B=0.5)
    Particles.Flash Flare Sprite Size:Vector2f=&Multiply_Vector2D_ByFloat(Float=&FN_FrontEndFloatModifier(Gameplay=1, Front End=$User.FrontEndScale), Vector2D=<200 450>)
   ) {isSetParameters=true, module=SetVariables_#4}
   &DynamicMaterialParameters(
    Index 0 Param 1=0.2
    Index 0 Param 2=1
    Module.Index0Param0Write=true
    Module.Index0Param1Write=true
    Module.Index0Param2Write=true
    Module.Index0Param3Write=true
   )
   &CameraOffset(Camera Offset Amount=15)
  ]
  particleUpdateScript [
   &ParticleState(DeltaTime:NiagaraFloat=&Multiply_Float(A=$Engine.DeltaTime, B=$System.SystemTimeScale))
   /SetVariables(
    Particles.Flash Flare Age:NiagaraFloat=&Divide_Float(A=$Particles.Age, B=$Particles.Initial.Flash Flare Lifetime)
    Particles.Flash Star Age:NiagaraFloat=&Divide_Float(A=$Particles.Age, B=$Particles.Initial.Flash Star Lifetime)
   ) {isSetParameters=true, module=SetVariables_#5}
   /SetVariables(
    Particles.Flash Flare Color:LinearColor=&MakeLinearColorFromVectorAndFloat(
     "Vector (RGB)"=&Multiply_Vector(
      A=&Multiply_Vector(
       A=&MakeVectorFromLinearColorRGB(Color=$Particles.Initial.Flash Flare Color)
       B=&MakeVectorFromLinearColorRGB(Color=&ColorFromCurve(
         ColorCurve=DataInterface() {propertyValues {
           AlphaCurve {keys [<at 0 1 RCIM_Cubic>, {arriveTangent=-1, interpMode=RCIM_Cubic, leaveTangent=-1, time=0.8, value=0.3}, <at 1 0>]}
           BlueCurve {keys [<at 0.0198 0.7369>,<at 0.4481 0.566>,<at 0.8999 2>]}
           GreenCurve {keys [<at 0.0198 1.335>,<at 0.4481 0.7576>,<at 0.8999 1.4232>]}
           RedCurve {keys [<at 0.0198 2>,<at 0.4481 2>,<at 0.8999 0.9029>]}
          }}
         CurveIndex=$Particles.Flash Flare Age
        ))
      )
      B=&VectorFromFloat(Value=&FN_FrontEndFloatModifier(Gameplay=1, Front End=0.75))
     )
     "Float (Alpha)"=&Multiply_Float(
      A=&MakeFloatFromLinearColor(LinearColor=<rgba 1 1 1 1>, Channel:ENiagaraLinearColor_Channels=::NewEnumerator3)
      B=&FloatFromCurve(
       FloatCurve=DataInterface() {propertyValues {Curve {keys [{arriveTangent=-1.9727, interpMode=RCIM_Cubic, time=0.1, value=1}, <at 1 0 RCIM_Cubic>]}}}
       CurveIndex=$Particles.Flash Flare Age
      )
     )
    )
    Particles.Flash Star Color:LinearColor=&MultiplyLinearColor(
     Linear Color=$Particles.Initial.Flash Star Color
     Scale Factor=&MakeVector4_FromLinearColor(LinearColor=&ColorFromCurve(
       ColorCurve=DataInterface() {propertyValues {
         AlphaCurve {keys [<at 0.2935 1>,<at 0.9862 0>]}
         BlueCurve {keys [<at 0 3>,<at 1 1>]}
         GreenCurve {keys [<at 0 1.3>,<at 1 0.2917>]}
         RedCurve {keys [<at 0 0.45>,<at 1 0.15>]}
        }}
       CurveIndex=$Particles.Flash Star Age
      ))
    )
    Particles.Flash Star Sprite Size:Vector2f=&Multiply_Vector2D_ByFloat(
     Float=&FloatFromCurve(
      FloatCurve=DataInterface() {propertyValues {Curve {keys [
          {arriveTangent=2.7013, interpMode=RCIM_Cubic, leaveTangent=2.7013, tangentMode=RCTM_User, time=0, value=0}
          {interpMode=RCIM_Cubic, tangentMode=RCTM_User, time=1, value=1}
         ]}}}
      CurveIndex=$Particles.Flash Star Age
     )
     Vector2D=$Particles.Initial.Flash Star Sprite Size
    )
   ) {isSetParameters=true, module=SetVariables_#6}
   &FN_Light_Attributes(
    Color=&MultiplyLinearColor(
     Linear Color=<rgba 0.15 0.3625 1 1>
     Scale Factor=&MakeVector4_FromLinearColor(LinearColor=&LerpLinearColors(StartColor=<rgba 1 1 1 1>, EndColor=$System.TOD Color, LerpFactor=2))
    )
    Create Color Attribute=true
    Create Exponent Attribute=true
    Create Position Attribute=true
    Create Radius Attribute=true
    Exponent=1
    Radius=&FN_FrontEndFloatModifier(Gameplay=90, Front End=90)
    Scale Color=&FloatFromCurve(
     FloatCurve=DataInterface() {propertyValues {Curve {keys [
         {arriveTangent=-0.5504, interpMode=RCIM_Cubic, leaveTangent=-0.5504, tangentMode=RCTM_User, time=0, value=1}
         {arriveTangent=-0.5426, interpMode=RCIM_Cubic, leaveTangent=-0.5426, tangentMode=RCTM_User, time=0.55, value=0}
        ]}}}
     Scale Curve=2
    )
   )
  ]
  rendererClasses [&NiagaraSpriteRendererProperties, &NiagaraLightRendererProperties]
  renderers [
   {rendererClass=&NiagaraSpriteRendererProperties}
   {rendererClass=&NiagaraSpriteRendererProperties, rendererIndex=1}
   {rendererClass=&NiagaraLightRendererProperties, rendererIndex=2}
  ]
 }
 FlashElements001 {
  emitterUpdateScript [
   &EmitterState(Life Cycle Mode=::NewEnumerator1, Loop Behavior=::NewEnumerator1, Loop Duration=0.06, UseLoopDelay=true)
   &SpawnBurst_Instantaneous(Spawn Group=1, Use Spawn Probability=false, UseLoopCountLimit=false) {module=SpawnBurst_Instantaneous001}
   &SpawnBurst_Instantaneous(
    Spawn Count=&RandomRangeInt(Minimum=6, Maximum=8)
    Use Spawn Probability=false
    UseLoopCountLimit=false
   )
  ]
  particleSpawnScript [
   &InitializeParticle(
    Color Mode=::NewEnumerator0
    Lifetime=0.25
    Position Offset=<0 0 10>
    Sprite Rotation Mode=::NewEnumerator2
    UsePositionOffset=true
    Write Material Random=true
   )
   &ShapeLocation(MaskSpawnGroup=false, Override Seed=false, Sphere Radius=1, UseOwnerScale=true)
   &SpriteFacingAndAlignment(Sprite Alignment=$Particles.ShapeLocation.ShapeNormal, WriteAlignment=true, WriteFacing=false)
   &FN_PerParticleLOD(Enable SpriteSize Distance Scaling=true, Sprite Size Scale Far=<2.5 2.5>)
   /SetVariables(Particles.SpriteSize=<1 1>, Particles.VisibilityTag=&SpawnGroup()) {isSetParameters=true, module=SetVariables_#2}
   /SetVariables(
    Particles.Flash Star Cam Offset=15
    Particles.Flash Star Lifetime=0.15
    Particles.Flash Star Rotation:NiagaraFloat=&RandomRangeFloat(Minimum=0, Maximum=360)
    Particles.Flash Star Sprite Size:Vector2f=&Multiply_Vector2D_ByFloat(
     Float=&Multiply_Float(
      A=&MakeCustomFloatFromBool(Bool=$User.bIsFrontEnd, True Float=$User.FrontEndScale, False Float=1.5)
      B=&RandomRangeFloat(Minimum=110, Maximum=130)
     )
     Vector2D=$Particles.SpriteSize
    )
    Particles.Flash Star SubUV Index:NiagaraFloat=&RandomRangeFloat(Minimum=0, Maximum=3)
   ) {isSetParameters=true, module=SetVariables_#3}
   /SetVariables(
    Particles.Flash Flare Lifetime:NiagaraFloat=&Multiply_Float(A=0.3, B=0.5)
    Particles.Flash Flare Sprite Size:Vector2f=&Multiply_Vector2D_ByFloat(Float=&FN_FrontEndFloatModifier(Gameplay=1, Front End=$User.FrontEndScale), Vector2D=<250 250>)
   ) {isSetParameters=true, module=SetVariables_#4}
   &DynamicMaterialParameters(
    Index 0 Param 1=0.2
    Index 0 Param 2=1
    Module.Index0Param0Write=true
    Module.Index0Param1Write=true
    Module.Index0Param2Write=true
    Module.Index0Param3Write=true
   )
   &CameraOffset(Camera Offset Amount=15)
  ]
  particleUpdateScript [
   &ParticleState(DeltaTime:NiagaraFloat=&Multiply_Float(A=$Engine.DeltaTime, B=$System.SystemTimeScale))
   /SetVariables(
    Particles.Flash Flare Age:NiagaraFloat=&Divide_Float(A=$Particles.Age, B=$Particles.Initial.Flash Flare Lifetime)
    Particles.Flash Star Age:NiagaraFloat=&Divide_Float(A=$Particles.Age, B=$Particles.Initial.Flash Star Lifetime)
   ) {isSetParameters=true, module=SetVariables_#5}
   /SetVariables(
    Particles.Flash Flare Color:LinearColor=&MakeLinearColorFromVectorAndFloat(
     "Vector (RGB)"=&Multiply_Vector(
      A=&Multiply_Vector(
       A=&MakeVectorFromLinearColorRGB(Color=$Particles.Initial.Flash Flare Color)
       B=&MakeVectorFromLinearColorRGB(Color=&ColorFromCurve(
         ColorCurve=DataInterface() {propertyValues {
           AlphaCurve {keys [<at 0 1 RCIM_Cubic>, {arriveTangent=-1, interpMode=RCIM_Cubic, leaveTangent=-1, time=0.8, value=0.3}, <at 1 0>]}
           BlueCurve {keys [<at 0.0198 0.7369>,<at 0.4481 0.566>,<at 0.8999 2>]}
           GreenCurve {keys [<at 0.0198 1.335>,<at 0.4481 0.7576>,<at 0.8999 1.4232>]}
           RedCurve {keys [<at 0.0198 2>,<at 0.4481 2>,<at 0.8999 0.9029>]}
          }}
         CurveIndex=$Particles.Flash Flare Age
        ))
      )
      B=&VectorFromFloat(Value=&FN_FrontEndFloatModifier(Gameplay=1, Front End=0.75))
     )
     "Float (Alpha)"=&Multiply_Float(
      A=&MakeFloatFromLinearColor(LinearColor=<rgba 1 1 1 1>, Channel:ENiagaraLinearColor_Channels=::NewEnumerator3)
      B=&FloatFromCurve(
       FloatCurve=DataInterface() {propertyValues {Curve {keys [{arriveTangent=-1.9727, interpMode=RCIM_Cubic, time=0.1, value=1}, <at 1 0 RCIM_Cubic>]}}}
       CurveIndex=$Particles.Flash Flare Age
      )
     )
    )
    Particles.Flash Star Color:LinearColor=&MultiplyLinearColor(
     Linear Color=$Particles.Initial.Flash Star Color
     Scale Factor=&MakeVector4_FromLinearColor(LinearColor=&ColorFromCurve(
       ColorCurve=DataInterface() {propertyValues {
         AlphaCurve {keys [<at 0.2935 1>,<at 0.9862 0>]}
         BlueCurve {keys [<at 0 3>,<at 1 1>]}
         GreenCurve {keys [<at 0 1.3>,<at 1 0.2917>]}
         RedCurve {keys [<at 0 0.45>,<at 1 0.15>]}
        }}
       CurveIndex=$Particles.Flash Star Age
      ))
    )
    Particles.Flash Star Sprite Size:Vector2f=&Multiply_Vector2D_ByFloat(
     Float=&FloatFromCurve(
      FloatCurve=DataInterface() {propertyValues {Curve {keys [
          {arriveTangent=2.7013, interpMode=RCIM_Cubic, leaveTangent=2.7013, tangentMode=RCTM_User, time=0, value=0}
          {interpMode=RCIM_Cubic, tangentMode=RCTM_User, time=1, value=1}
         ]}}}
      CurveIndex=$Particles.Flash Star Age
     )
     Vector2D=$Particles.Initial.Flash Star Sprite Size
    )
   ) {isSetParameters=true, module=SetVariables_#6}
   &FN_Light_Attributes(
    Color=&MultiplyLinearColor(
     Linear Color=<rgba 1 0.5112 0.1765 1>
     Scale Factor=&MakeVector4_FromLinearColor(LinearColor=&LerpLinearColors(StartColor=<rgba 1 1 1 1>, EndColor=$System.TOD Color, LerpFactor=2))
    )
    Create Color Attribute=true
    Create Exponent Attribute=true
    Create Position Attribute=true
    Create Radius Attribute=true
    Exponent=1
    Radius=&FN_FrontEndFloatModifier(Gameplay=90, Front End=90)
    Scale Color=&FloatFromCurve(
     FloatCurve=DataInterface() {propertyValues {Curve {keys [
         {arriveTangent=-0.5504, interpMode=RCIM_Cubic, leaveTangent=-0.5504, tangentMode=RCTM_User, time=0, value=1}
         {arriveTangent=-0.5426, interpMode=RCIM_Cubic, leaveTangent=-0.5426, tangentMode=RCTM_User, time=0.55, value=0}
        ]}}}
     Scale Curve=2
    )
   )
  ]
  rendererClasses [&NiagaraSpriteRendererProperties, &NiagaraLightRendererProperties]
  renderers [
   {rendererClass=&NiagaraSpriteRendererProperties}
   {rendererClass=&NiagaraSpriteRendererProperties, rendererIndex=1}
   {rendererClass=&NiagaraLightRendererProperties, rendererIndex=2}
  ]
 }
 FlashElements002 {
  emitterUpdateScript [
   &EmitterState(Life Cycle Mode=::NewEnumerator1, Loop Behavior=::NewEnumerator1, Loop Duration=0.06, UseLoopDelay=true)
   &SpawnBurst_Instantaneous(Spawn Group=1, Use Spawn Probability=false, UseLoopCountLimit=false) {module=SpawnBurst_Instantaneous001}
   &SpawnBurst_Instantaneous(
    Spawn Count=&RandomRangeInt(Minimum=6, Maximum=8)
    Use Spawn Probability=false
    UseLoopCountLimit=false
   )
  ]
  particleSpawnScript [
   &InitializeParticle(
    Color Mode=::NewEnumerator0
    Position Offset=<0 0 10>
    Sprite Rotation Mode=::NewEnumerator2
    UsePositionOffset=true
    Write Material Random=true
   )
   &ShapeLocation(MaskSpawnGroup=false, Override Seed=false, Sphere Radius=1, UseOwnerScale=true)
   &SpriteFacingAndAlignment(Sprite Alignment=$Particles.ShapeLocation.ShapeNormal, WriteAlignment=true, WriteFacing=false)
   &FN_PerParticleLOD(Enable SpriteSize Distance Scaling=true, Sprite Size Scale Far=<2.5 2.5>)
   /SetVariables(Particles.SpriteSize=<1 1>, Particles.VisibilityTag=&SpawnGroup()) {isSetParameters=true, module=SetVariables_#2}
   /SetVariables(
    Particles.Flash Star Cam Offset=15
    Particles.Flash Star Lifetime=0.15
    Particles.Flash Star Rotation:NiagaraFloat=&RandomRangeFloat(Minimum=0, Maximum=360)
    Particles.Flash Star Sprite Size:Vector2f=&Multiply_Vector2D_ByFloat(
     Float=&Multiply_Float(
      A=&MakeCustomFloatFromBool(Bool=$User.bIsFrontEnd, True Float=$User.FrontEndScale, False Float=1.5)
      B=&RandomRangeFloat(Minimum=110, Maximum=130)
     )
     Vector2D=$Particles.SpriteSize
    )
    Particles.Flash Star SubUV Index:NiagaraFloat=&RandomRangeFloat(Minimum=0, Maximum=3)
   ) {isSetParameters=true, module=SetVariables_#3}
   /SetVariables(
    Particles.Flash Flare Lifetime:NiagaraFloat=&Multiply_Float(A=0.3, B=0.5)
    Particles.Flash Flare Sprite Size:Vector2f=&Multiply_Vector2D_ByFloat(Float=&FN_FrontEndFloatModifier(Gameplay=1, Front End=$User.FrontEndScale), Vector2D=<250 250>)
   ) {isSetParameters=true, module=SetVariables_#4}
   &DynamicMaterialParameters(
    Index 0 Param 1=0.2
    Index 0 Param 2=1
    Module.Index0Param0Write=true
    Module.Index0Param1Write=true
    Module.Index0Param2Write=true
    Module.Index0Param3Write=true
   )
   &CameraOffset(Camera Offset Amount=15)
  ]
  particleUpdateScript [
   &ParticleState(DeltaTime:NiagaraFloat=&Multiply_Float(A=$Engine.DeltaTime, B=$System.SystemTimeScale))
   /SetVariables(
    Particles.Flash Flare Age:NiagaraFloat=&Divide_Float(A=$Particles.Age, B=$Particles.Initial.Flash Flare Lifetime)
    Particles.Flash Star Age:NiagaraFloat=&Divide_Float(A=$Particles.Age, B=$Particles.Initial.Flash Star Lifetime)
   ) {isSetParameters=true, module=SetVariables_#5}
   /SetVariables(
    Particles.Flash Flare Color:LinearColor=&MakeLinearColorFromVectorAndFloat(
     "Vector (RGB)"=&Multiply_Vector(
      A=&Multiply_Vector(
       A=&MakeVectorFromLinearColorRGB(Color=$Particles.Initial.Flash Flare Color)
       B=&MakeVectorFromLinearColorRGB(Color=&ColorFromCurve(
         ColorCurve=DataInterface() {propertyValues {
           AlphaCurve {keys [<at 0 1 RCIM_Cubic>, {arriveTangent=-1, interpMode=RCIM_Cubic, leaveTangent=-1, time=0.8, value=0.3}, <at 1 0>]}
           BlueCurve {keys [<at 0.0198 0.7369>,<at 0.4481 0.566>,<at 0.8999 2>]}
           GreenCurve {keys [<at 0.0198 1.335>,<at 0.4481 0.7576>,<at 0.8999 1.4232>]}
           RedCurve {keys [<at 0.0198 2>,<at 0.4481 2>,<at 0.8999 0.9029>]}
          }}
         CurveIndex=$Particles.Flash Flare Age
        ))
      )
      B=&VectorFromFloat(Value=&FN_FrontEndFloatModifier(Gameplay=1, Front End=0.75))
     )
     "Float (Alpha)"=&Multiply_Float(
      A=&MakeFloatFromLinearColor(LinearColor=<rgba 1 1 1 1>, Channel:ENiagaraLinearColor_Channels=::NewEnumerator3)
      B=&FloatFromCurve(
       FloatCurve=DataInterface() {propertyValues {Curve {keys [{arriveTangent=-1.9727, interpMode=RCIM_Cubic, time=0.1, value=1}, <at 1 0 RCIM_Cubic>]}}}
       CurveIndex=$Particles.Flash Flare Age
      )
     )
    )
    Particles.Flash Star Color:LinearColor=&MultiplyLinearColor(
     Linear Color=$Particles.Initial.Flash Star Color
     Scale Factor=&MakeVector4_FromLinearColor(LinearColor=&ColorFromCurve(
       ColorCurve=DataInterface() {propertyValues {
         AlphaCurve {keys [<at 0.2935 1>,<at 0.9862 0>]}
         BlueCurve {keys [<at 0 3>,<at 1 1>]}
         GreenCurve {keys [<at 0 1.3>,<at 1 0.2917>]}
         RedCurve {keys [<at 0 0.45>,<at 1 0.15>]}
        }}
       CurveIndex=$Particles.Flash Star Age
      ))
    )
    Particles.Flash Star Sprite Size:Vector2f=&Multiply_Vector2D_ByFloat(
     Float=&FloatFromCurve(
      FloatCurve=DataInterface() {propertyValues {Curve {keys [
          {arriveTangent=2.7013, interpMode=RCIM_Cubic, leaveTangent=2.7013, tangentMode=RCTM_User, time=0, value=0}
          {interpMode=RCIM_Cubic, tangentMode=RCTM_User, time=1, value=1}
         ]}}}
      CurveIndex=$Particles.Flash Star Age
     )
     Vector2D=$Particles.Initial.Flash Star Sprite Size
    )
   ) {isSetParameters=true, module=SetVariables_#6}
   &FN_Light_Attributes(
    Color=&MultiplyLinearColor(
     Linear Color=&ScaleLinearColorByCurve(
      InitialValue=<rgba 1 1 1 1>
      ScaleCurve=DataInterface() {propertyValues {
        AlphaCurve {keys []}
        BlueCurve {keys [<at 0 2.3884>,<at 0.1644 0.4986>,<at 0.8624 0.177>,<at 1 0.0673>]}
        GreenCurve {keys [<at 0 8.0363>,<at 0.1644 1.4127>,<at 0.8624 0.3002>,<at 1 0.0388>]}
        RedCurve {keys [<at 0 15>,<at 0.1644 5>,<at 0.8624 2.5781>,<at 1 0.4968>]}
       }}
      CurveIndex=$Particles.NormalizedAge
     )
     Scale Factor=&MakeVector4_FromLinearColor(LinearColor=&LerpLinearColors(StartColor=<rgba 1 1 1 1>, EndColor=$System.TOD Color, LerpFactor=2))
    )
    Create Color Attribute=true
    Create Exponent Attribute=true
    Create Position Attribute=true
    Create Radius Attribute=true
    Exponent=2
    Position=<0 0 50>
    Radius=&FN_FrontEndFloatModifier(Gameplay=150, Front End=90)
    Scale Color=&FloatFromCurve(FloatCurve=DataInterface() {propertyValues {Curve {keys [
         {arriveTangent=-0.5504, interpMode=RCIM_Cubic, leaveTangent=-0.5504, tangentMode=RCTM_User, time=0, value=1}
         {arriveTangent=-2, interpMode=RCIM_Cubic, leaveTangent=-2, time=0.3, value=0.0547}
         {arriveTangent=-0.5426, interpMode=RCIM_Cubic, leaveTangent=-0.5426, tangentMode=RCTM_User, time=0.5, value=0}
        ]}}})
   )
  ]
  rendererClasses [&NiagaraSpriteRendererProperties, &NiagaraLightRendererProperties]
  renderers [
   {rendererClass=&NiagaraSpriteRendererProperties}
   {rendererClass=&NiagaraSpriteRendererProperties, rendererIndex=1}
   {rendererClass=&NiagaraLightRendererProperties, rendererIndex=2}
  ]
 }
 Sparks {
  emitterUpdateScript [
   &EmitterState(Life Cycle Mode=::NewEnumerator1, Loop Behavior=::NewEnumerator1, Loop Duration=0.25, UseLoopDelay=false)
   &SpawnRate(SpawnRate=600, Use Spawn Probability=false)
   &SpawnBurst_Instantaneous(Spawn Count=&RandomRangeInt(Minimum=1200, Maximum=1200), Use Spawn Probability=false)
  ]
  particleSpawnScript [
   &InitializeParticle(
    Color=&RandomRangeLinearColor(Minimum=<rgba 10 2.4082 0 1>, Maximum=<rgba 30 7.2245 0 1>)
    Lifetime=/Niagara/DynamicInputs/Helpers/DistributionCurveFloat(
     Minimum:NiagaraFloat=0.25
     Maximum:NiagaraFloat=1.1
     Distribution Curve:NiagaraDataInterfaceCurve=DataInterface() {propertyValues {Curve {defaultValue=0, keys [{tangentMode=RCTM_User, time=0, value=0}, <at 0.8 0>, {tangentMode=RCTM_User, time=1, value=1}]}}}
     Randomness Mode:ENiagaraRandomnessMode=::NewEnumerator0
     Evaluation Type:ENiagaraRandomnessEvaluation=::NewEnumerator0
     Seed:NiagaraInt32=0
     Fixed Random Seed:NiagaraBool=false
    )
    Sprite Size Max=<1 5>
    Sprite Size Min=<1 1>
    Sprite Size Mode=::NewEnumerator2
    UsePositionOffset=false
    Write Material Random=true
   ) {module=InitializeParticle001}
   &ShapeLocation(
    MaskSpawnGroup=false
    Override Seed=false
    Sphere Radius=&RandomRangeFloat(Minimum=20, Maximum=50)
    UseOwnerScale=true
   )
   &AddVelocity(
    Module.Use Falloff Distance=false
    Override Seed=false
    Velocity Mode=::NewEnumerator1
    Velocity Speed=&RandomRangeFloat(Minimum=500, Maximum=300)
   )
   &AddVelocity(Velocity=<0 0 400>) {module=AddVelocity001}
  ]
  particleUpdateScript [
   &ParticleState(DeltaTime:NiagaraFloat=&Multiply_Float(A=$Engine.DeltaTime, B=$System.SystemTimeScale))
   &ScaleSpriteSize(Uniform Curve Sprite Scale=DataInterface() {propertyValues {Curve {
       defaultValue=0
       keys [
        {tangentMode=RCTM_User, time=0, value=1}
        {arriveTangent=-0.0405, interpMode=RCIM_Cubic, leaveTangent=-0.0405, tangentMode=RCTM_User, time=0.7009, value=1}
        {arriveTangent=-8.0089, interpMode=RCIM_Cubic, leaveTangent=-8.0089, tangentMode=RCTM_User, time=1, value=0}
       ]
      }}})
   &Drag(Drag=&RandomRangeFloat(Minimum=2, Maximum=5), Use Linear Drag=true, Use Rotational Drag=false)
   &GravityForce(Gravity=&VectorFromCurve(
     VectorCurve=DataInterface() {propertyValues {
       XCurve {keys [<at 0 1>,<at 1 0>]}
       YCurve {keys [<at 0 1>,<at 1 0>]}
       ZCurve {defaultValue=1, keys [{tangentMode=RCTM_User, time=0, value=0}, <at 0.2 0>, {tangentMode=RCTM_User, time=0.4, value=1}]}
      }}
     Scale Curve=<0 0 -980>
    ))
   &SolveForcesAndVelocity(
    Clamp Velocity=false
    DeltaTime:NiagaraFloat=&Multiply_Float(A=$Engine.DeltaTime, B=$System.SystemTimeScale)
    Limit Acceleration=false
   )
   &ScaleSpriteSizeBySpeed(
    Max Scale Factor=<0.6 15>
    Min Scale Factor=<0.2 1>
    Sample Scale Factor By Curve=true
    Scale Factor Curve=DataInterface() {propertyValues {Curve {keys [
        <at -0.0077 0.0765 RCIM_Cubic>
        {arriveTangent=1.2463, interpMode=RCIM_Cubic, leaveTangent=1.2463, tangentMode=RCTM_User, time=0.9985, value=0.9549}
       ]}}}
   ) {enabled=false}
   &ScaleColor(
    Scale Alpha=&FloatFromCurve(FloatCurve=DataInterface() {propertyValues {Curve {
        defaultValue=0
        keys [{interpMode=RCIM_Cubic, tangentMode=RCTM_User, time=0, value=1}, {interpMode=RCIM_Cubic, tangentMode=RCTM_User, time=1, value=0}]
       }}})
    ScaleA=true
    ScaleRGB=true
   )
   /SetVariables(
    Particles.Linear Color=&MultiplyLinearColor(Linear Color=$Particles.Color, Scale Factor=<0.1 0 0 0.05>)
    Particles.Vector 2D:Vector2f=&Multiply_Vector2D(A=$Particles.SpriteSize, B=<10 2>)
   ) {isSetParameters=true}
  ]
  rendererClasses [&NiagaraSpriteRendererProperties]
  renderers [
   {rendererClass=&NiagaraSpriteRendererProperties}
   {rendererClass=&NiagaraSpriteRendererProperties, rendererIndex=1}
  ]
 }
 snow_mist {
  emitterSpawnScript [/SetVariables(Emitter.FE Particle Spawn Rate=35, Emitter.Idle Spawn Rate=60, Emitter.Max Traversal Speed=150, Emitter.Particle Spawn Rate=60) {enabled=false, isSetParameters=true}]
  emitterUpdateScript [
   &EmitterState(Life Cycle Mode=::NewEnumerator1, Loop Behavior=::NewEnumerator1, Loop Duration=1, UseLoopDelay=false)
   &SpawnBurst_Instantaneous(Spawn Count=15, Use Spawn Probability=false, UseLoopCountLimit=false)
   &SpawnRate(SpawnRate=60, Use Spawn Probability=false) {module=SpawnRate001}
   &SpawnRate(
    SpawnRate=&FN_FrontEndFloatModifier(
     Gameplay=&MakeCustomFloatFromBool(
      Bool=&SetBoolByFloatComparison(A=&VectorLength(Vector=$User.OwnerLocalVelocity), B=300)
      True Float=$Emitter.Particle Spawn Rate
      False Float=$Emitter.Idle Spawn Rate
     )
     Front End=$Emitter.FE Particle Spawn Rate
    )
    Use Spawn Probability=false
   ) {enabled=false}
   /SetVariables(Emitter.EmitterColor:LinearColor=&ColorFromCurve(
     ColorCurve=DataInterface() {propertyValues {
       AlphaCurve {keys [<at 0 1 RCIM_Cubic>,<at 1 0 RCIM_Cubic>]}
       BlueCurve {keys [<at 0 1.1434>,<at 0.0034 0.2473>,<at 0.1431 1 RCIM_Cubic>]}
       GreenCurve {keys [<at 0 4.8938>,<at 0.0034 0.5973>,<at 0.1431 1 RCIM_Cubic>]}
       RedCurve {keys [<at 0 10>,<at 0.0034 1>,<at 0.1431 1 RCIM_Cubic>]}
      }}
     CurveIndex=&Divide_Float(A=$Emitter.Age, B=2)
    )) {enabled=false, isSetParameters=true}
  ]
  particleSpawnScript [
   &InitializeParticle(
    Color=&MultiplyLinearColor(
     Linear Color=&FNTimeOfDayColor(
      Night Color=<rgba 0.4036 0.4922 0.901 1>
      Morning Color=<rgba 0.7325 0.7699 0.8177 1>
      Day Color=<rgba 0.7325 0.7699 0.8177 1>
      Evening Color=<rgba 0.7325 0.7699 0.8177 1>
     )
     Scale Factor=&MakeVector4(X=1, Y=1, Z=1, W=&Multiply_Float(A=1, B=$System.SystemTimeScale))
    )
    Lifetime=&Multiply_Float(A=&RandomRangeFloat(Minimum=0.8, Maximum=1), B=0.7)
    Sprite Rotation Mode=::NewEnumerator2
    Sprite Size Mode=::NewEnumerator1
    Uniform Sprite Size Max=250
    Uniform Sprite Size Min=100
    UsePositionOffset=false
    Write Material Random=true
   )
   &TorusLocation(
    Align Torus With Axis=false
    Handle Radius=0
    Large Radius:NiagaraFloat=&Multiply_Float(A=60, B=&ClampFloat(Float=$User.ForwardDot, Min=0.3))
    MaskSpawnGroup=false
    Module.Override Randomness=false
   )
   &AddVelocityFromPoint(
    Module.Use Falloff Distance=false
    Velocity Strength=&Multiply_Float(
     A=800
     B=&FloatFromCurve(FloatCurve=DataInterface() {propertyValues {Curve {keys [<at 0 0.5>,<at 0.1 1>,<at 1 0>]}}})
    )
   )
   &AddVelocity(Velocity=<0 0 350>)
   &AddVelocityInCone(Cone Angle=10, Cone Axis=<0 0 1>, Module.Use Velocity Falloff On Cone Axis=false, Velocity Strength=150) {module=AddVelocityInCone001}
   &CameraOffset(Camera Offset Amount=10)
   &DynamicMaterialParameters(
    Index 0 Param 1=1
    Index 0 Param 2=1
    Index 0 Param 3=1
    Index 0 Param 4=1
    Module.Param0WriteEnabled=true
    Module.Param1WriteEnabled=true
    Module.Param2WriteEnabled=true
    Module.Param3WriteEnabled=true
   )
  ]
  particleUpdateScript [
   &ParticleState(DeltaTime:NiagaraFloat=&Multiply_Float(A=$Engine.DeltaTime, B=$System.SystemTimeScale))
   &AccelerationForce(Acceleration=&MakeVector(X=300, Y=0, Z=300))
   &GravityForce(Gravity=&RandomRangeVector(Minimum=<0 0 -300>, Maximum=<1 1 -500>))
   &Drag(Drag=&RandomRangeFloat(Minimum=1, Maximum=10), Use Linear Drag=true, Use Rotational Drag=true)
   &SolveForcesAndVelocity(
    Clamp Velocity=false
    DeltaTime:NiagaraFloat=&Multiply_Float(A=$Engine.DeltaTime, B=$System.SystemTimeScale)
    Limit Acceleration=false
   )
   &ScaleSpriteSize(
    Scale Sprite Size Mode=::NewEnumerator0
    Uniform Scale Factor=&FloatFromCurve(
     FloatCurve=DataInterface() {propertyValues {Curve {keys [
         {arriveTangent=0.8753, interpMode=RCIM_Cubic, leaveTangent=0.8753, tangentMode=RCTM_User, time=0, value=0.6}
         {arriveTangent=0.0238, interpMode=RCIM_Cubic, leaveTangent=0.0238, tangentMode=RCTM_User, time=1, value=0.8}
        ]}}}
     Scale Curve=0.75
    )
   )
   &ScaleColor(
    Scale Alpha=&Multiply_Float(
     A=&FloatFromCurve(
      FloatCurve=DataInterface() {propertyValues {Curve {keys [
          {interpMode=RCIM_Cubic, tangentMode=RCTM_User, time=0, value=0}
          {interpMode=RCIM_Cubic, leaveTangent=-2.6773, tangentMode=RCTM_Break, time=0.1, value=1}
          {tangentMode=RCTM_User, time=1, value=0}
         ]}}}
      Scale Curve=&RandomRangeFloat(Minimum=0.5, Maximum=1)
     )
     B=&FNTimeOfDayScalar(Night Scale=0.6, Morning Scale=0.7, Evening Scale=0.5)
    )
    ScaleA=true
    ScaleRGB=true
   )
   &SubUVAnimation(SubUV Loop Count=0.3)
   &BlendSpace("Position Blend (World <-> Local)":NiagaraFloat=&MakeCustomFloatFromBool(
     Bool=&SetBoolByFloatComparison(A=&VectorLength(Vector=$Engine.Owner.Velocity), B=400)
     True Float=1
     False Float=&FloatFromCurve(FloatCurve=DataInterface() {propertyValues {Curve {keys [<at 0 1 RCIM_Cubic>, {arriveTangent=-3.5186, interpMode=RCIM_Cubic, leaveTangent=-3.5186, tangentMode=RCTM_User, time=1, value=0}]}}})
     Translation Contribution:NiagaraFloat=1
     Rotation Contribution:NiagaraFloat=1
    )) {enabled=false}
   /SetVariables(Particles.Color=&MultiplyLinearColor(
     Linear Color=$Particles.Color
     Scale Factor=&MakeVector4_FromLinearColor(LinearColor=$Emitter.EmitterColor)
    )) {isSetParameters=true}
  ]
  rendererClasses [&NiagaraSpriteRendererProperties]
  renderers [{rendererClass=&NiagaraSpriteRendererProperties}]
 }
}
summary {
 systemName=NS_Emote_WavySignalBeef_Explosion_New
 emitters [
  {emitterName=snow_mist, rendererClasses [&NiagaraSpriteRendererProperties]}
  {emitterName=Sparks, rendererClasses [&NiagaraSpriteRendererProperties]}
  {
   emitterName=FlashElements001
   rendererClasses [&NiagaraSpriteRendererProperties, &NiagaraLightRendererProperties]
  }
  {emitterName=FlashElements, rendererClasses [&NiagaraSpriteRendererProperties, &NiagaraLightRendererProperties]}
  {
   emitterName=FlashElements002
   rendererClasses [&NiagaraSpriteRendererProperties, &NiagaraLightRendererProperties]
  }
 ]
 userVariables [
  User.AnimTimeScale {type=NiagaraFloat, defaultValue=1, description=""}
  User.bIsFrontEnd {type=NiagaraBool}
  User.bIsFrontEndPreview {type=NiagaraBool}
  User.BurstIndex {type=NiagaraFloat, defaultValue=0, description=""}
  User.ForwardDot {type=NiagaraFloat, defaultValue=0, description=""}
  User.FrontEndScale {type=NiagaraFloat, defaultValue=0.6, description=""}
  User.NormalizedNotifyProgress {type=NiagaraFloat, defaultValue=0, description=""}
  User.OwnerLocalVelocity {type=Vector3f, description="", defaultValue {struct=Vector3f, value=<0 0 0>}}
  User.Speed {type=NiagaraFloat, defaultValue=0, description=""}
 ]
}
