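# Niagara effect for the WavySignalBeef emote that fires a short flash with rays, glows, sparks and mist.
# A flash emitter drives a light with front-end-adjusted radius, alongside burst emitters for sparks and mist.
# Ray and glow emitters pick palettes by spawn group and shift camera offset with view angle.
# The glow emitter feeds five sprite renderers, computing separate flare, glow and light-ray colors and size curves per particle.
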NiagaraSystem /BRCosmetics/Animation/Game/MainPlayer/Emotes/WavySignalBeef/FX/NS_Emote_WavySignalBeef_Flash1
dependencies {
 usedDataInterfaces [
  &NiagaraDataInterfaceArrayColor
  &NiagaraDataInterfaceCurve
  &NiagaraDataInterfaceVectorCurve
  &NiagaraDataInterfaceCamera
  &NiagaraDataInterfaceColorCurve
  &NiagaraDataInterfaceVector2DCurve
 ]
 usedDynamicInputs [&Multiply_VectorByFloat]
 usedModules [&SystemState]
}
emitters {
 FlashElements002 {
  emitterUpdateScript [
   &EmitterState(Life Cycle Mode=::NewEnumerator1, Loop Behavior=::NewEnumerator1, Loop Duration=0.06, UseLoopDelay=false)
   &SpawnBurst_Instantaneous(Spawn Group=1, Use Spawn Probability=false, UseLoopCountLimit=false) {module=SpawnBurst_Instantaneous001}
   &SpawnBurst_Instantaneous(
    Spawn Count=&RandomRangeInt(Minimum=6, Maximum=8)
    Use Spawn Probability=false
    UseLoopCountLimit=false
   )
  ]
  particleSpawnScript [
   &InitializeParticle(
    Color Mode=::NewEnumerator0
    Lifetime=0.4
    Position Offset=<0 0 10>
    Sprite Rotation Mode=::NewEnumerator2
    UsePositionOffset=true
    Write Material Random=true
   )
   &ShapeLocation(MaskSpawnGroup=false, Override Seed=false, Sphere Radius=1, UseOwnerScale=true)
   &SpriteFacingAndAlignment(Sprite Alignment=$Particles.ShapeLocation.ShapeNormal, WriteAlignment=true, WriteFacing=false)
   &FN_PerParticleLOD(Enable SpriteSize Distance Scaling=true, Sprite Size Scale Far=<2.5 2.5>)
   /SetVariables(Particles.SpriteSize=<1 1>, Particles.VisibilityTag=&SpawnGroup()) {isSetParameters=true}
   /SetVariables(
    Particles.Flash Star Cam Offset=15
    Particles.Flash Star Lifetime=0.15
    Particles.Flash Star Rotation:NiagaraFloat=&RandomRangeFloat(Minimum=0, Maximum=360)
    Particles.Flash Star Sprite Size:Vector2f=&Multiply_Vector2D_ByFloat(
     Float=&Multiply_Float(
      A=&MakeCustomFloatFromBool(Bool=$User.bIsFrontEnd, True Float=$User.FrontEndScale, False Float=1.5)
      B=&RandomRangeFloat(Minimum=110, Maximum=130)
     )
     Vector2D=$Particles.SpriteSize
    )
    Particles.Flash Star SubUV Index:NiagaraFloat=&RandomRangeFloat(Minimum=0, Maximum=3)
   ) {isSetParameters=true}
   /SetVariables(
    Particles.Flash Flare Lifetime:NiagaraFloat=&Multiply_Float(A=0.3, B=0.5)
    Particles.Flash Flare Sprite Size:Vector2f=&Multiply_Vector2D_ByFloat(Float=&FN_FrontEndFloatModifier(Gameplay=1, Front End=$User.FrontEndScale), Vector2D=<250 250>)
   ) {isSetParameters=true}
   &DynamicMaterialParameters(
    Index 0 Param 1=0.2
    Index 0 Param 2=1
    Module.Index0Param0Write=true
    Module.Index0Param1Write=true
    Module.Index0Param2Write=true
    Module.Index0Param3Write=true
   )
   &CameraOffset(Camera Offset Amount=15)
  ]
  particleUpdateScript [
   &ParticleState()
   /SetVariables(
    Particles.Flash Flare Age:NiagaraFloat=&Divide_Float(A=$Particles.Age, B=$Particles.Initial.Flash Flare Lifetime)
    Particles.Flash Star Age:NiagaraFloat=&Divide_Float(A=$Particles.Age, B=$Particles.Initial.Flash Star Lifetime)
   ) {isSetParameters=true}
   /SetVariables(
    Particles.Flash Flare Color:LinearColor=&MakeLinearColorFromVectorAndFloat(
     "Vector (RGB)"=&Multiply_Vector(
      A=&Multiply_Vector(
       A=&MakeVectorFromLinearColorRGB(Color=$Particles.Initial.Flash Flare Color)
       B=&MakeVectorFromLinearColorRGB(Color=&ColorFromCurve(
         ColorCurve=DataInterface() {propertyValues {
           AlphaCurve {keys [<at 0 1 RCIM_Cubic>, {arriveTangent=-1, interpMode=RCIM_Cubic, leaveTangent=-1, time=0.8, value=0.3}, <at 1 0>]}
           BlueCurve {keys [<at 0.0198 0.7369>,<at 0.4481 0.566>,<at 0.8999 2>]}
           GreenCurve {keys [<at 0.0198 1.335>,<at 0.4481 0.7576>,<at 0.8999 1.4232>]}
           RedCurve {keys [<at 0.0198 2>,<at 0.4481 2>,<at 0.8999 0.9029>]}
          }}
         CurveIndex=$Particles.Flash Flare Age
        ))
      )
      B=&VectorFromFloat(Value=&FN_FrontEndFloatModifier(Gameplay=1, Front End=0.75))
     )
     "Float (Alpha)"=&Multiply_Float(
      A=&MakeFloatFromLinearColor(LinearColor=<rgba 1 1 1 1>, Channel:ENiagaraLinearColor_Channels=::NewEnumerator3)
      B=&FloatFromCurve(
       FloatCurve=DataInterface() {propertyValues {Curve {keys [{arriveTangent=-1.9727, interpMode=RCIM_Cubic, time=0.1, value=1}, <at 1 0 RCIM_Cubic>]}}}
       CurveIndex=$Particles.Flash Flare Age
      )
     )
    )
    Particles.Flash Star Color:LinearColor=&MultiplyLinearColor(
     Linear Color=$Particles.Initial.Flash Star Color
     Scale Factor=&MakeVector4_FromLinearColor(LinearColor=&ColorFromCurve(
       ColorCurve=DataInterface() {propertyValues {
         AlphaCurve {keys [<at 0.2935 1>,<at 0.9862 0>]}
         BlueCurve {keys [<at 0 3>,<at 1 1>]}
         GreenCurve {keys [<at 0 1.3>,<at 1 0.2917>]}
         RedCurve {keys [<at 0 0.45>,<at 1 0.15>]}
        }}
       CurveIndex=$Particles.Flash Star Age
      ))
    )
    Particles.Flash Star Sprite Size:Vector2f=&Multiply_Vector2D_ByFloat(
     Float=&FloatFromCurve(
      FloatCurve=DataInterface() {propertyValues {Curve {keys [
          {arriveTangent=2.7013, interpMode=RCIM_Cubic, leaveTangent=2.7013, tangentMode=RCTM_User, time=0, value=0}
          {interpMode=RCIM_Cubic, tangentMode=RCTM_User, time=1, value=1}
         ]}}}
      CurveIndex=$Particles.Flash Star Age
     )
     Vector2D=$Particles.Initial.Flash Star Sprite Size
    )
   ) {isSetParameters=true}
   &FN_Light_Attributes(
    Color=&MultiplyLinearColor(
     Linear Color=&ScaleLinearColorByCurve(
      InitialValue=<rgba 1 1 1 1>
      ScaleCurve=DataInterface() {propertyValues {
        AlphaCurve {keys []}
        BlueCurve {keys [<at 0 2.3884>,<at 0.1644 0.4986>,<at 0.8624 0.4443>,<at 1 0.1379>]}
        GreenCurve {keys [<at 0 8.0363>,<at 0.1644 1.4127>,<at 0.8624 1.334>,<at 1 0.3869>]}
        RedCurve {keys [<at 0 15>,<at 0.1644 5>,<at 0.8624 2.5781>,<at 1 0.4968>]}
       }}
      CurveIndex=$Particles.NormalizedAge
     )
     Scale Factor=&MakeVector4_FromLinearColor(LinearColor=&LerpLinearColors(StartColor=<rgba 1 1 1 1>, EndColor=$System.TOD Color, LerpFactor=2))
    )
    Create Color Attribute=true
    Create Exponent Attribute=true
    Create Position Attribute=true
    Create Radius Attribute=true
    Exponent=2
    Radius=&FN_FrontEndFloatModifier(Gameplay=150, Front End=90)
    Scale Color=&FloatFromCurve(FloatCurve=DataInterface() {propertyValues {Curve {keys [
         {arriveTangent=-0.5504, interpMode=RCIM_Cubic, leaveTangent=-0.5504, tangentMode=RCTM_User, time=0, value=1}
         {arriveTangent=-1, interpMode=RCIM_Cubic, leaveTangent=-1, time=0.3, value=0.0547}
         {arriveTangent=-0.5426, interpMode=RCIM_Cubic, leaveTangent=-0.5426, tangentMode=RCTM_User, time=1, value=0}
        ]}}})
   )
  ]
  rendererClasses [&NiagaraLightRendererProperties]
  renderers [{rendererClass=&NiagaraLightRendererProperties, rendererIndex=0}]
 }
 Ray_01 {
  emitterUpdateScript [
   &EmitterState(Life Cycle Mode=::NewEnumerator1, Loop Behavior=::NewEnumerator1, Loop Duration=0.08, UseLoopDelay=false)
   &SpawnRate(SpawnGroup:NiagaraInt32=$User.SpawnGroup, SpawnRate=70, Use Spawn Probability=false)
   &SpawnBurst_Instantaneous(Spawn Count=15, Spawn Group:NiagaraInt32=$User.SpawnGroup, Spawn Time=0.03, Use Spawn Probability=false, UseLoopCountLimit=false)
  ]
  particleSpawnScript [
   &PartitionParticles()
   &ShapeLocation(
    MaskSpawnGroup=false
    Override Seed=false
    Rotation Mode=::NewEnumerator1
    Sphere Radius=&RandomRangeFloat(Minimum=0, Maximum=50)
    UseOwnerScale=true
   ) {enabled=false, module=ShapeLocation001}
   &InitializeParticle(
    Color=&MergeHelper_ColorSwitch(
     0=&SelectLinearColorFromArray(
      Array Sampling Mode=::NewEnumerator1
      Direct Array Index=$Particles.PartitionParticles.Partition
      Color Selection Array=DataInterface() {propertyValues {ColorData [<rgba 0.0017 0.0057 0.0104 1>, <rgba 0.9635 0.7942 0.2108 1>]}}
      Random Seed=?
     )
     1=&SelectLinearColorFromArray(
      Array Sampling Mode=::NewEnumerator1
      Direct Array Index=$Particles.PartitionParticles.Partition
      Color Selection Array=DataInterface() {propertyValues {ColorData [<rgba 0.0365 0.0365 0.0365 1>, <rgba 1 1 1 1>]}}
      Random Seed=?
     )
     2=&SelectLinearColorFromArray(
      Array Sampling Mode=::NewEnumerator1
      Direct Array Index=$Particles.PartitionParticles.Partition
      Color Selection Array=DataInterface() {propertyValues {ColorData [<rgba 3 0.1564 0 1>, <rgba 0.396 0 0 1>]}}
      Random Seed=?
     )
     3=&SelectLinearColorFromArray(
      Array Sampling Mode=::NewEnumerator1
      Direct Array Index=$Particles.PartitionParticles.Partition
      Color Selection Array=DataInterface() {propertyValues {ColorData [<rgba 0.599 0.2832 0.0076 1>, <rgba 2.5 1.2273 0 1>]}}
      Random Seed=?
     )
     4=&SelectLinearColorFromArray(
      Array Sampling Mode=::NewEnumerator1
      Direct Array Index=$Particles.PartitionParticles.Partition
      Color Selection Array=DataInterface() {propertyValues {ColorData [<rgba 0 0.0312 0.4531 1>, <rgba 0.0144 0.1721 0.3958 1>]}}
      Random Seed=?
     )
     5=&SelectLinearColorFromArray(
      Array Sampling Mode=::NewEnumerator1
      Direct Array Index=$Particles.PartitionParticles.Partition
      Color Selection Array=DataInterface() {propertyValues {ColorData [<rgba 0.4033 0.0005 0.4375 1>, <rgba 1 0 0.6697 1>]}}
      Random Seed=?
     )
     6=&SelectLinearColorFromArray(
      Array Sampling Mode=::NewEnumerator1
      Direct Array Index=$Particles.PartitionParticles.Partition
      Color Selection Array=DataInterface() {propertyValues {ColorData [<rgba 0 0.3854 0.0733 1>, <rgba 0.2324 1.5 0 1>]}}
      Random Seed=?
     )
     Index=&SpawnGroup()
     Num=::NewEnumerator5
    )
    Lifetime Max=0.35
    Lifetime Min=0.15
    Lifetime Mode=::NewEnumerator1
    Sprite Size=&RandomRangeVector2D(Minimum=<10 120>, Maximum=<80 140>)
    Sprite Size Mode=::NewEnumerator4
    UsePositionOffset=false
    Write Material Random=true
   )
   &CameraOffset(Camera Offset Amount=-5) {enabled=false}
   &SubUVAnimation(End Frame=1, SubUV Animation Mode:ENiagara_SubUVLookupMode=::NewEnumerator2)
   /SetVariables(Particles.Ray2Size:Vector2f=&RandomRangeVector2D(Minimum=<50 100>, Maximum=<80 120>)) {isSetParameters=true}
   &ShapeLocation(
    MaskSpawnGroup=false
    Override Seed=false
    Ring Radius=10
    Rotation Angle=0.5
    Rotation Axis=<0 1 0>
    Rotation Mode=::NewEnumerator1
    Shape Primitive=::NewEnumerator4
    UseOwnerScale=true
   )
   &SpriteFacingAndAlignment(Sprite Alignment=$Particles.ShapeLocation.ShapeNormal, WriteAlignment=true, WriteFacing=true) {enabled=false}
   &Color(Color=&MaskLinearColorBySpawnGroup(
     Masked Linear Color=&SelectLinearColorFromArray(
      Array Sampling Mode=::NewEnumerator1
      Direct Array Index=$Particles.PartitionParticles.Partition
      Color Selection Array=DataInterface() {propertyValues {ColorData [<rgba 0.0365 0.0365 0.0365 1>, <rgba 1 1 1 1>]}}
      Random Seed=?
     )
     Passthrough Linear Color=&MaskLinearColorBySpawnGroup(
      Masked Linear Color=&SelectLinearColorFromArray(
       Array Sampling Mode=::NewEnumerator1
       Direct Array Index=$Particles.PartitionParticles.Partition
       Color Selection Array=DataInterface() {propertyValues {ColorData [<rgba 3 0.1564 0 1>, <rgba 0.396 0 0 1>]}}
       Random Seed=?
      )
      Passthrough Linear Color=&MaskLinearColorBySpawnGroup(
       Masked Linear Color=&SelectLinearColorFromArray(
        Array Sampling Mode=::NewEnumerator1
        Direct Array Index=$Particles.PartitionParticles.Partition
        Color Selection Array=DataInterface() {propertyValues {ColorData [<rgba 0.599 0.2832 0.0076 1>, <rgba 2.5 1.2273 0 1>]}}
        Random Seed=?
       )
       Passthrough Linear Color=&MaskLinearColorBySpawnGroup(
        Masked Linear Color=&SelectLinearColorFromArray(
         Array Sampling Mode=::NewEnumerator1
         Direct Array Index=$Particles.PartitionParticles.Partition
         Color Selection Array=DataInterface() {propertyValues {ColorData [<rgba 0 0.0312 0.4531 1>, <rgba 0.0144 0.1721 0.3958 1>]}}
         Random Seed=?
        )
        Passthrough Linear Color=&MaskLinearColorBySpawnGroup(
         Masked Linear Color=&SelectLinearColorFromArray(
          Array Sampling Mode=::NewEnumerator1
          Direct Array Index=$Particles.PartitionParticles.Partition
          Color Selection Array=DataInterface() {propertyValues {ColorData [<rgba 0.4033 0.0005 0.4375 1>, <rgba 1 0 0.6697 1>]}}
          Random Seed=?
         )
         Passthrough Linear Color=&MaskLinearColorBySpawnGroup(Masked Linear Color=<rgba 0.0884 1 0 1>, Passthrough Linear Color=<rgba 1 1 1 1>, Spawn Group Mask=6)
         Spawn Group Mask=5
        )
        Spawn Group Mask=4
       )
       Spawn Group Mask=3
      )
      Spawn Group Mask=2
     )
     Spawn Group Mask=1
    )) {enabled=false}
   &AddVelocity(Module.Use Falloff Distance=false, Override Seed=false, Velocity Mode=::NewEnumerator1, Velocity Speed=0.1)
  ]
  particleUpdateScript [
   &ParticleState()
   &ScaleSpriteSize(
    Scale Factor=&MakeVector2D(
     X=&FloatFromCurve(FloatCurve=DataInterface() {propertyValues {Curve {keys [<at 0 0.5>]}}})
     Y=&FloatFromCurve(FloatCurve=DataInterface() {propertyValues {Curve {keys [<at 0 0.5>,<at 1 1>]}}})
    )
    Scale Sprite Size Mode=::NewEnumerator2
   )
   &ScaleColor(
    Scale Alpha=&Multiply_Float(
     A=&FloatFromCurve(FloatCurve=DataInterface() {propertyValues {Curve {keys [{interpMode=RCIM_Cubic, tangentMode=RCTM_User, time=0, value=0}, {interpMode=RCIM_Cubic, tangentMode=RCTM_User, time=0.4, value=1}, <at 1 0>]}}})
     B=&FloatFromCurve(
      FloatCurve=DataInterface() {propertyValues {Curve {keys [
          <at 0 0.5>
          {arriveTangent=0.1737, interpMode=RCIM_Cubic, leaveTangent=0.1737, tangentMode=RCTM_User, time=0.2, value=1.35}
          {interpMode=RCIM_Cubic, tangentMode=RCTM_User, time=0.6, value=1.3}
          {arriveTangent=-0.5397, interpMode=RCIM_Cubic, leaveTangent=-0.5397, tangentMode=RCTM_User, time=0.8, value=0.6}
          {arriveTangent=-0.405, interpMode=RCIM_Cubic, leaveTangent=-0.405, tangentMode=RCTM_User, time=1.3, value=0}
         ]}}}
      CurveIndex=$Engine.System.Age
     )
    )
    ScaleA=true
    ScaleRGB=true
   )
   &SolveForcesAndVelocity(Clamp Velocity=false, Limit Acceleration=false)
   &CameraOffset(Camera Offset Amount=&RemapRange(
     Input Value To Remap=&MakeFloatFromVector(
      Vector=&VectorFromCurve(
       VectorCurve=DataInterface() {propertyValues {XCurve {keys [<at 0 1>,<at 1 0>]}, YCurve {keys [<at 0 1>,<at 1 0>]}, ZCurve {keys [<at 0 1>,<at 1 0>]}}}
       CurveIndex=&DotProduct(
        Vector A=&ReturnCameraVector(
         Camera Vectors=::NewEnumerator4
         Coordinate Space=::Simulation
         Origin Position=$Particles.Position
         CameraQuery=DataInterface() {propertyValues {bRequireCurrentFrameData=false}}
        )
        Vector B=<0 1 0>
       )
      )
      Channel=::NewEnumerator0
     )
     Input Min=0.2
     Input Max=0.9
     Output Min=-15
     Output Max=40
    )) {module=CameraOffset001}
  ]
  rendererClasses [&NiagaraSpriteRendererProperties]
  renderers [
   {rendererClass=&NiagaraSpriteRendererProperties}
   {rendererClass=&NiagaraSpriteRendererProperties, rendererIndex=1}
  ]
 }
 Ray_01001 {
  emitterUpdateScript [
   &EmitterState(Life Cycle Mode=::NewEnumerator1, Loop Behavior=::NewEnumerator1, Loop Duration=1.0167, UseLoopDelay=false)
   &SpawnBurst_Instantaneous(Spawn Count=3, Spawn Group:NiagaraInt32=$User.SpawnGroup, Use Spawn Probability=false, UseLoopCountLimit=false)
   &SpawnBurst_Instantaneous(Spawn Count=3, Spawn Time=1.0167, Use Spawn Probability=false, UseLoopCountLimit=false) {module=SpawnBurst_Instantaneous001}
  ]
  particleSpawnScript [
   &InitializeParticle(
    Color=&MergeHelper_ColorSwitch(
     0=<rgba 0 0 0 0>
     1=<rgba 1 1 1 1>
     2=<rgba 1 0 0 1>
     3=<rgba 0.0479 0 1 1>
     4=<rgba 0 1 0.2398 1>
     5=<rgba 1 0 0.5228 1>
     6=<rgba 0 0 0 0>
     Index=&SpawnGroup()
     Num=::NewEnumerator5
    )
    Lifetime=0.45
    Sprite Size Mode=::NewEnumerator3
    Uniform Sprite Size=100
    UsePositionOffset=false
    Write Material Random=true
   )
   &Color(Color=&MaskLinearColorBySpawnGroup(
     Masked Linear Color=&ColorFromCurve(ColorCurve=DataInterface() {propertyValues {
        AlphaCurve {keys [<at 0.4804 0.6 RCIM_Cubic>,<at 1 0 RCIM_Cubic>]}
        BlueCurve {keys [<at 0 0.6562 RCIM_Cubic>]}
        GreenCurve {keys [<at 0 0.6562 RCIM_Cubic>]}
        RedCurve {keys [<at 0 0.6562 RCIM_Cubic>]}
       }})
     Passthrough Linear Color=&MaskLinearColorBySpawnGroup(
      Masked Linear Color=&MakeLinearColorFromVectorAndFloat(
       "Vector (RGB)"=&MakeVectorFromLinearColorRGB(Color=<rgba 1.2 0.2149 0.0267 0.8087>)
       "Float (Alpha)"=&FloatFromCurve(FloatCurve=DataInterface() {propertyValues {Curve {keys [<at 0.7 0.7>,<at 0.9521 0>]}}})
      )
      Passthrough Linear Color=&MaskLinearColorBySpawnGroup(
       Masked Linear Color=&MakeLinearColorFromVectorAndFloat(
        "Vector (RGB)"=&MakeVectorFromLinearColorRGB(Color=<rgba 1.2 0.5216 0.0451 0.7391>)
        "Float (Alpha)"=&FloatFromCurve(FloatCurve=DataInterface() {propertyValues {Curve {keys [<at 0.7 0.7422>,<at 1 0>]}}})
       )
       Passthrough Linear Color=&MaskLinearColorBySpawnGroup(
        Masked Linear Color=&MakeLinearColorFromVectorAndFloat(
         "Vector (RGB)"=&MakeVectorFromLinearColorRGB(Color=<rgba 0.1315 0.4845 2 1>)
         "Float (Alpha)"=&FloatFromCurve(FloatCurve=DataInterface() {propertyValues {Curve {keys [<at 0.7 0.7342>,<at 1 0>]}}})
        )
        Passthrough Linear Color=&MaskLinearColorBySpawnGroup(
         Masked Linear Color=&MakeLinearColorFromVectorAndFloat(
          "Vector (RGB)"=&MakeVectorFromLinearColorRGB(Color=&MakeLinearColorFromVectorAndFloat("Vector (RGB)"=&MakeVectorFromLinearColorRGB(Color=<rgba 2 0.0519 1.005 0.4957>), "Float (Alpha)"=1))
          "Float (Alpha)"=&FloatFromCurve(FloatCurve=DataInterface() {propertyValues {Curve {keys [<at 0.7 0.75>,<at 1 0>]}}})
         )
         Passthrough Linear Color=&MaskLinearColorBySpawnGroup(Masked Linear Color=<rgba 0.1817 1 0 1>, Passthrough Linear Color=<rgba 1 1 1 1>, Spawn Group Mask=6)
         Spawn Group Mask=5
        )
        Spawn Group Mask=4
       )
       Spawn Group Mask=3
      )
      Spawn Group Mask=2
     )
     Spawn Group Mask=1
    )) {enabled=false}
   &SpriteFacingAndAlignment(
    Alignment Coordinate Space=::Local
    Facing Coordinate Space=::Local
    Sprite Alignment=<1 0 0>
    Sprite Facing=<0 1 0>
    WriteAlignment=true
    WriteFacing=true
   )
  ]
  particleUpdateScript [
   &ParticleState()
   &ScaleSpriteSize(Scale Sprite Size Mode=::NewEnumerator2)
   &ScaleColor(
    Scale Alpha=&Multiply_Float(
     A=&FloatFromCurve(FloatCurve=DataInterface() {propertyValues {Curve {keys [{interpMode=RCIM_Cubic, tangentMode=RCTM_User, time=0, value=0}, {interpMode=RCIM_Cubic, tangentMode=RCTM_User, time=0.4, value=1}, <at 1 0>]}}})
     B=&FloatFromCurve(
      FloatCurve=DataInterface() {propertyValues {Curve {keys [
          <at 0 0.5>
          {arriveTangent=0.1737, interpMode=RCIM_Cubic, leaveTangent=0.1737, tangentMode=RCTM_User, time=0.2, value=1.35}
          {interpMode=RCIM_Cubic, tangentMode=RCTM_User, time=0.6, value=1.3}
          {arriveTangent=-0.5397, interpMode=RCIM_Cubic, leaveTangent=-0.5397, tangentMode=RCTM_User, time=0.8, value=0.6}
          {arriveTangent=-0.405, interpMode=RCIM_Cubic, leaveTangent=-0.405, tangentMode=RCTM_User, time=1.3, value=0}
         ]}}}
      CurveIndex=$Engine.System.Age
     )
    )
    ScaleA=true
    ScaleRGB=true
   )
   &SolveForcesAndVelocity(Clamp Velocity=false, Limit Acceleration=false)
   /SetVariables(
    Particles.FlareColor=&MergeHelper_ColorSwitch(
     0=<rgba 0.0637 0.0701 0.0833 1>
     1=<rgba 1 0.9279 0.2856 1>
     2=<rgba 1.5 0.2887 0.1199 1>
     3=<rgba 1 0.4383 0 1>
     4=<rgba 0.2425 0.7056 1.5 1>
     5=<rgba 1.3683 0.3922 1.5 1>
     6=<rgba 0.1418 1.138 0 1>
     Index=&SpawnGroup()
     Num=::NewEnumerator5
    )
    Particles.GlowBlockerColor:LinearColor=&MergeHelper_ColorSwitch(
     0=&ColorFromCurve(ColorCurve=DataInterface() {propertyValues {
        AlphaCurve {keys [<at 0.4804 1 RCIM_Cubic>,<at 1 0 RCIM_Cubic>]}
        BlueCurve {keys [<at 0 0.4834 RCIM_Cubic>]}
        GreenCurve {keys [<at 0 0.8506 RCIM_Cubic>]}
        RedCurve {keys [<at 0 1 RCIM_Cubic>]}
       }})
     1=&ColorFromCurve(ColorCurve=DataInterface() {propertyValues {
        AlphaCurve {keys [<at 0.4804 0.6 RCIM_Cubic>,<at 1 0 RCIM_Cubic>]}
        BlueCurve {keys [<at 0 0.6562 RCIM_Cubic>]}
        GreenCurve {keys [<at 0 0.6562 RCIM_Cubic>]}
        RedCurve {keys [<at 0 0.6562 RCIM_Cubic>]}
       }})
     2=&MakeLinearColorFromVectorAndFloat(
      "Vector (RGB)"=&MakeVectorFromLinearColorRGB(Color=<rgba 1.2 0.2149 0.0267 0.8087>)
      "Float (Alpha)"=&FloatFromCurve(FloatCurve=DataInterface() {propertyValues {Curve {keys [<at 0.7 0.7>,<at 0.9521 0>]}}})
     )
     3=&MakeLinearColorFromVectorAndFloat(
      "Vector (RGB)"=&MakeVectorFromLinearColorRGB(Color=<rgba 1.2 0.5216 0.0451 0.7391>)
      "Float (Alpha)"=&FloatFromCurve(FloatCurve=DataInterface() {propertyValues {Curve {keys [<at 0.7 0.7422>,<at 1 0>]}}})
     )
     4=&MakeLinearColorFromVectorAndFloat(
      "Vector (RGB)"=&MakeVectorFromLinearColorRGB(Color=<rgba 0.1315 0.4845 2 1>)
      "Float (Alpha)"=&FloatFromCurve(FloatCurve=DataInterface() {propertyValues {Curve {keys [<at 0.7 0.7342>,<at 1 0>]}}})
     )
     5=&MakeLinearColorFromVectorAndFloat(
      "Vector (RGB)"=&MakeVectorFromLinearColorRGB(Color=&MakeLinearColorFromVectorAndFloat("Vector (RGB)"=&MakeVectorFromLinearColorRGB(Color=<rgba 2 0.0519 1.005 0.4957>), "Float (Alpha)"=1))
      "Float (Alpha)"=&FloatFromCurve(FloatCurve=DataInterface() {propertyValues {Curve {keys [<at 0.7 0.75>,<at 1 0>]}}})
     )
     6=&MakeLinearColorFromVectorAndFloat(
      "Vector (RGB)"=&MakeVectorFromLinearColorRGB(Color=&MakeLinearColorFromVectorAndFloat("Vector (RGB)"=&MakeVectorFromLinearColorRGB(Color=<rgba 0.2565 1.3229 0 0.4957>), "Float (Alpha)"=1))
      "Float (Alpha)"=&FloatFromCurve(FloatCurve=DataInterface() {propertyValues {Curve {keys [<at 0.7 0.75>,<at 1 0>]}}})
     )
     Index=&SpawnGroup()
     Num=::NewEnumerator5
    )
    Particles.GlowColor=&MergeHelper_ColorSwitch(
     0=&ColorFromCurve(ColorCurve=DataInterface() {propertyValues {
        AlphaCurve {keys [<at 0.4804 0.6 RCIM_Cubic>,<at 1 0 RCIM_Cubic>]}
        BlueCurve {keys [<at 0 0.026 RCIM_Cubic>]}
        GreenCurve {keys [<at 0 0.026 RCIM_Cubic>]}
        RedCurve {keys [<at 0 0.026 RCIM_Cubic>]}
       }})
     1=&ColorFromCurve(ColorCurve=DataInterface() {propertyValues {
        AlphaCurve {keys [<at 0.4804 0.6 RCIM_Cubic>,<at 1 0 RCIM_Cubic>]}
        BlueCurve {keys [<at 0 0.6562 RCIM_Cubic>]}
        GreenCurve {keys [<at 0 0.6562 RCIM_Cubic>]}
        RedCurve {keys [<at 0 0.6562 RCIM_Cubic>]}
       }})
     2=&MakeLinearColorFromVectorAndFloat(
      "Vector (RGB)"=&MakeVectorFromLinearColorRGB(Color=<rgba 1.2 0.0393 0.0156 0.8087>)
      "Float (Alpha)"=&FloatFromCurve(FloatCurve=DataInterface() {propertyValues {Curve {keys [<at 0.7 0.5>,<at 1 -0.0557>]}}})
     )
     3=&MakeLinearColorFromVectorAndFloat(
      "Vector (RGB)"=&MakeVectorFromLinearColorRGB(Color=<rgba 1.2 0.3412 0.0106 0.7391>)
      "Float (Alpha)"=&FloatFromCurve(FloatCurve=DataInterface() {propertyValues {Curve {keys [<at 0.7 0.2>,<at 1 0>]}}})
     )
     4=&MakeLinearColorFromVectorAndFloat(
      "Vector (RGB)"=&MakeVectorFromLinearColorRGB(Color=<rgba 0.3035 0.1625 2 1>)
      "Float (Alpha)"=&FloatFromCurve(FloatCurve=DataInterface() {propertyValues {Curve {keys [<at 0.7 0.25>,<at 1 0>]}}})
     )
     5=&MakeLinearColorFromVectorAndFloat(
      "Vector (RGB)"=&MakeVectorFromLinearColorRGB(Color=&MakeLinearColorFromVectorAndFloat("Vector (RGB)"=&MakeVectorFromLinearColorRGB(Color=<rgba 2 0.0519 1.005 0.4957>), "Float (Alpha)"=1))
      "Float (Alpha)"=&FloatFromCurve(FloatCurve=DataInterface() {propertyValues {Curve {keys [<at 0.7 0.2>,<at 1 0>]}}})
     )
     6=&MakeLinearColorFromVectorAndFloat(
      "Vector (RGB)"=&MakeVectorFromLinearColorRGB(Color=&MakeLinearColorFromVectorAndFloat("Vector (RGB)"=&MakeVectorFromLinearColorRGB(Color=<rgba 0.0364 2 0 0.4957>), "Float (Alpha)"=1))
      "Float (Alpha)"=&FloatFromCurve(FloatCurve=DataInterface() {propertyValues {Curve {keys [<at 0.7 0.2>,<at 1 0>]}}})
     )
     Index=&SpawnGroup()
     Num=::NewEnumerator5
    )
    Particles.LensFlareColor:LinearColor=&MergeHelper_ColorSwitch(
     0=<rgba 3 2.1299 1.4442 0.113>
     1=<rgba 3 3 3 0.113>
     2=&ColorFromCurve(ColorCurve=DataInterface() {propertyValues {
        AlphaCurve {keys [<at 0 0.493 RCIM_Cubic>,<at 1 0 RCIM_Cubic>]}
        BlueCurve {keys [<at 0 1 RCIM_Cubic>,<at 0.1115 0.1416 RCIM_Cubic>]}
        GreenCurve {keys [<at 0 1 RCIM_Cubic>,<at 0.1115 0.1784 RCIM_Cubic>]}
        RedCurve {keys [<at 0 1 RCIM_Cubic>,<at 0.1115 2 RCIM_Cubic>]}
       }})
     3=&ColorFromCurve(ColorCurve=DataInterface() {propertyValues {
        AlphaCurve {keys [<at 0 0.493 RCIM_Cubic>,<at 1 0 RCIM_Cubic>]}
        BlueCurve {keys [<at 0 1 RCIM_Cubic>,<at 0.1115 0.0634 RCIM_Cubic>]}
        GreenCurve {keys [<at 0 1 RCIM_Cubic>,<at 0.1115 0.452 RCIM_Cubic>]}
        RedCurve {keys [<at 0 1 RCIM_Cubic>,<at 0.1115 2 RCIM_Cubic>]}
       }})
     4=&ColorFromCurve(ColorCurve=DataInterface() {propertyValues {
        AlphaCurve {keys [<at 0 0.453 RCIM_Cubic>,<at 1 0 RCIM_Cubic>]}
        BlueCurve {keys [<at 0 1 RCIM_Cubic>,<at 0.1115 2.5 RCIM_Cubic>]}
        GreenCurve {keys [<at 0 1 RCIM_Cubic>,<at 0.1115 0.932 RCIM_Cubic>]}
        RedCurve {keys [<at 0 1 RCIM_Cubic>,<at 0.1115 0.1286 RCIM_Cubic>]}
       }})
     5=&ColorFromCurve(ColorCurve=DataInterface() {propertyValues {
        AlphaCurve {keys [<at 0 0.55 RCIM_Cubic>,<at 1 0 RCIM_Cubic>]}
        BlueCurve {keys [<at 0 1 RCIM_Cubic>,<at 0.1115 2.1643 RCIM_Cubic>]}
        GreenCurve {keys [<at 0 1 RCIM_Cubic>,<at 0.1115 0.1183 RCIM_Cubic>]}
        RedCurve {keys [<at 0 1 RCIM_Cubic>,<at 0.1115 2.5 RCIM_Cubic>]}
       }})
     6=&ColorFromCurve(ColorCurve=DataInterface() {propertyValues {
        AlphaCurve {keys [<at 0 0.55 RCIM_Cubic>,<at 1 0 RCIM_Cubic>]}
        BlueCurve {keys [<at 0 1 RCIM_Cubic>,<at 0.1115 0.0051 RCIM_Cubic>]}
        GreenCurve {keys [<at 0 1 RCIM_Cubic>,<at 0.1115 1.6304 RCIM_Cubic>]}
        RedCurve {keys [<at 0 1 RCIM_Cubic>,<at 0.1115 0.1229 RCIM_Cubic>]}
       }})
     Index=&SpawnGroup()
     Num=::NewEnumerator5
    )
    Particles.LightRayColor:LinearColor=&MergeHelper_ColorSwitch(
     0=<rgba 0.0175 0.0013 0.0417 0.2087>
     1=<rgba 0 0 0 0.7304>
     2=&FNTimeOfDayColor(
      Night Color=<rgba 0.224 0.0312 0.0312 1>
      Morning Color=<rgba 0.1458 0 0 1>
      Day Color=<rgba 0.4427 0 0 1>
      Evening Color=<rgba 0.1458 0 0 1>
     )
     3=<rgba 0.4115 0.2516 0.0921 0.3739>
     4=&FNTimeOfDayColor(
      Night Color=<rgba 0.052 0.1994 0.401 0.887>
      Morning Color=<rgba 0.0668 0.0993 0.349 1>
      Day Color=<rgba 0.0668 0.0993 0.349 1>
      Evening Color=<rgba 0.0668 0.0993 0.349 1>
     )
     5=<rgba 0.375 0.0879 0.2104 0.5304>
     6=<rgba 0 0.1979 0.0165 0.5913>
     Index=&SpawnGroup()
     Num=::NewEnumerator5
    )
   ) {isSetParameters=true}
   /SetVariables(
    Particles.DynParm0:Vector4f=&MakeVector4(
     X=&FloatFromCurve(FloatCurve=DataInterface() {propertyValues {Curve {keys [<at 0 0.3529>,<at 1 0>]}}})
     Y=1.9467
     Z=1.607
     W=3.6012
    )
    Particles.DynParm1:Vector4f=&MakeVector4(
     X=&FloatFromCurve(FloatCurve=DataInterface() {propertyValues {Curve {keys [<at 0 0.2339>,<at 1 0>]}}})
     Y=0
     Z=0.032
     W=1.2308
    )
    Particles.DynParm2:Vector4f=&MakeVector4(X=3, Y=0.87, Z=2.4276, W=0.86)
   ) {isSetParameters=true}
   /SetVariables(
    Particles.FlareSize:Vector2f=&Vector2DFromCurve(
     Vector2Curve=DataInterface() {propertyValues {
       XCurve {keys [<at 0.0391 0 RCIM_Constant>, <at 0.05 1 RCIM_Constant>, {interpMode=RCIM_Cubic, tangentMode=RCTM_User, time=1, value=1}]}
       YCurve {
        defaultValue=0
        keys [
         {arriveTangent=-6.7592, interpMode=RCIM_Cubic, leaveTangent=-6.7592, tangentMode=RCTM_User, time=0.05, value=2}
         {arriveTangent=-3.5504, interpMode=RCIM_Cubic, leaveTangent=-3.5504, tangentMode=RCTM_User, time=0.2381, value=0.2606}
         {arriveTangentWeight=0.3333, interpMode=RCIM_Cubic, tangentMode=RCTM_User, time=0.5211, value=0}
        ]
       }
      }}
     ScaleCurve=<228 228>
    )
    Particles.GlowBlockerSize:Vector2f=&Vector2DFromCurve(
     Vector2Curve=DataInterface() {propertyValues {
       XCurve {keys [
         {arriveTangent=-0.3851, interpMode=RCIM_Cubic, leaveTangent=-0.3851, tangentMode=RCTM_User, time=0.0144, value=0}
         {arriveTangent=0.2684, leaveTangent=0.1204, time=0.2391, value=1}
         <at 0.2391 1>
         {arriveTangent=-1.5463, leaveTangent=-1.5463, time=0.8977, value=1}
         <at 1 1>
        ]}
       YCurve {keys [
         {arriveTangent=-0.3851, arriveTangentWeight=0.0048, interpMode=RCIM_Cubic, leaveTangent=-0.3851, tangentMode=RCTM_User, time=0.0144, value=0}
         <at 0.2391 1>
         <at 0.8977 1>
         <at 1 1>
        ]}
      }}
     ScaleCurve=&Multiply_Vector2D_ByFloat(Float=1.5, Vector2D=<60 60>)
    )
    Particles.GlowSize=&Vector2DFromCurve(
     Vector2Curve=DataInterface() {propertyValues {
       XCurve {keys [
         {arriveTangent=-0.3851, interpMode=RCIM_Cubic, leaveTangent=-0.3851, tangentMode=RCTM_User, time=0.0144, value=0}
         <at 0.3533 1>
         {arriveTangent=0.2684, leaveTangent=0.1204, time=0.3533, value=1}
         {arriveTangent=-1.5463, leaveTangent=-1.5463, time=0.8977, value=1}
         <at 1 1>
        ]}
       YCurve {keys [<at 0 1>,<at 0.8977 1>,<at 1 1>]}
      }}
     ScaleCurve=<110 110>
    )
    Particles.LenFlareSize:Vector2f=&Vector2DFromCurve(
     Vector2Curve=DataInterface() {propertyValues {
       XCurve {
        defaultValue=1
        keys [
         <at 0 0.5 RCIM_Cubic>
         <at 0.03 1 RCIM_Constant>
         {interpMode=RCIM_Cubic, tangentMode=RCTM_Break, time=0.06, value=1}
         {arriveTangent=2.5018, arriveTangentWeight=0.0294, interpMode=RCIM_Cubic, leaveTangent=2.5018, tangentMode=RCTM_User, time=0.2032, value=0}
        ]
       }
       YCurve {
        defaultValue=0
        keys [
         <at 0 0.5 RCIM_Cubic>
         {arriveTangent=2.1725, interpMode=RCIM_Constant, time=0.03, value=1}
         {interpMode=RCIM_Cubic, tangentMode=RCTM_Break, time=0.06, value=1}
         {arriveTangent=2.5018, arriveTangentWeight=0.3333, interpMode=RCIM_Cubic, leaveTangent=2.5018, tangentMode=RCTM_User, time=0.2032, value=0}
        ]
       }
      }}
     ScaleCurve=<150 150>
    )
    Particles.LightRaySize:Vector2f=&Vector2DFromCurve(
     Vector2Curve=DataInterface() {propertyValues {
       XCurve {keys [
         {arriveTangent=-0.3851, interpMode=RCIM_Cubic, leaveTangent=-0.3851, tangentMode=RCTM_User, time=0.05, value=0}
         {arriveTangent=0.2684, interpMode=RCIM_Cubic, leaveTangent=-1.6898, tangentMode=RCTM_Break, time=0.1859, value=1.1681}
         <at 0.6186 1>
         <at 0.8739 0>
        ]}
       YCurve {keys [<at 0 1>,<at 0.6629 1>,<at 0.8739 0>]}
      }}
     ScaleCurve=&Multiply_Vector2D_ByFloat(Float=1.5, Vector2D=<168 168>)
    )
   ) {isSetParameters=true}
   &CameraOffset(Camera Offset Amount=&RemapRange(
     Input Value To Remap=&MakeFloatFromVector(
      Vector=&VectorFromCurve(
       VectorCurve=DataInterface() {propertyValues {XCurve {keys [<at 0 1>,<at 1 0>]}, YCurve {keys [<at 0 1>,<at 1 0>]}, ZCurve {keys [<at 0 1>,<at 1 0>]}}}
       CurveIndex=&DotProduct(
        Vector A=&ReturnCameraVector(
         Camera Vectors=::NewEnumerator4
         Coordinate Space=::Simulation
         Origin Position=$Particles.Position
         CameraQuery=DataInterface() {propertyValues {bRequireCurrentFrameData=false}}
        )
        Vector B=<0 1 0>
       )
      )
      Channel=::NewEnumerator0
     )
     Input Min=0.2
     Input Max=0.9
     Output Min=-15
     Output Max=20
    )) {module=CameraOffset001}
  ]
  rendererClasses [&NiagaraSpriteRendererProperties]
  renderers [
   {rendererClass=&NiagaraSpriteRendererProperties}
   {rendererClass=&NiagaraSpriteRendererProperties, rendererIndex=1}
   {rendererClass=&NiagaraSpriteRendererProperties, rendererIndex=2}
   {rendererClass=&NiagaraSpriteRendererProperties, rendererIndex=3}
   {rendererClass=&NiagaraSpriteRendererProperties, rendererIndex=4}
  ]
 }
 Sparks001 {
  emitterUpdateScript [
   &EmitterState(Life Cycle Mode=::NewEnumerator1, Loop Behavior=::NewEnumerator1, Loop Duration=0.25, UseLoopDelay=false)
   &SpawnRate(SpawnRate=600, Use Spawn Probability=false) {enabled=false}
   &SpawnBurst_Instantaneous(Spawn Count=&RandomRangeInt(Minimum=1000, Maximum=1000), Use Spawn Probability=false)
  ]
  particleSpawnScript [
   &InitializeParticle(
    Color=&RandomRangeLinearColor(Minimum=<rgba 10 5.9569 0.4957 1>, Maximum=<rgba 30 22.8307 14.1395 1>)
    Lifetime=/Niagara/DynamicInputs/Helpers/DistributionCurveFloat(
     Minimum:NiagaraFloat=0.25
     Maximum:NiagaraFloat=0.8
     Distribution Curve:NiagaraDataInterfaceCurve=DataInterface() {propertyValues {Curve {defaultValue=0, keys [{tangentMode=RCTM_User, time=0, value=0}, <at 0.8 0>, {tangentMode=RCTM_User, time=1, value=1}]}}}
     Randomness Mode:ENiagaraRandomnessMode=::NewEnumerator0
     Evaluation Type:ENiagaraRandomnessEvaluation=::NewEnumerator0
     Seed:NiagaraInt32=0
     Fixed Random Seed:NiagaraBool=false
    )
    Sprite Size Max=<1 5>
    Sprite Size Min=<1 1>
    Sprite Size Mode=::NewEnumerator2
    UsePositionOffset=false
    Write Material Random=true
   ) {module=InitializeParticle001}
   &ShapeLocation(
    Cylinder Height=0
    Cylinder Height Midpoint=0
    Cylinder Radius=1
    MaskSpawnGroup=false
    Override Seed=false
    Rotation Axis=<0 1 0>
    Rotation Mode=::NewEnumerator1
    Shape Primitive=::NewEnumerator1
    Surface Only=true
    UseOwnerScale=true
   )
   &AddVelocity(
    Module.Use Falloff Distance=false
    Override Seed=false
    Velocity Mode=::NewEnumerator1
    Velocity Speed=&RandomRangeFloat(Minimum=500, Maximum=300)
   )
   &AddVelocity(Velocity=<0 0 400>) {enabled=false, module=AddVelocity001}
   &CameraOffset(Camera Offset Amount=-90)
  ]
  particleUpdateScript [
   &ParticleState()
   &ScaleSpriteSize(Uniform Curve Sprite Scale=DataInterface() {propertyValues {Curve {
       defaultValue=0
       keys [
        {tangentMode=RCTM_User, time=0, value=1}
        {arriveTangent=-0.0405, interpMode=RCIM_Cubic, leaveTangent=-0.0405, tangentMode=RCTM_User, time=0.7009, value=1}
        {arriveTangent=-8.0089, interpMode=RCIM_Cubic, leaveTangent=-8.0089, tangentMode=RCTM_User, time=1, value=0}
       ]
      }}})
   &Drag(Drag=&RandomRangeFloat(Minimum=2, Maximum=5), Use Linear Drag=true, Use Rotational Drag=false)
   &GravityForce(Gravity=&VectorFromCurve(
     VectorCurve=DataInterface() {propertyValues {
       XCurve {keys [<at 0 0>,<at 1 0>]}
       YCurve {keys [<at 0 0>,<at 1 0>]}
       ZCurve {defaultValue=1, keys [{tangentMode=RCTM_User, time=0, value=0}, <at 0.2 0>, {tangentMode=RCTM_User, time=0.4, value=1}]}
      }}
     Scale Curve=<0 0 -980>
    ))
   &SolveForcesAndVelocity(Clamp Velocity=false, Limit Acceleration=false)
   &ScaleSpriteSizeBySpeed(
    Max Scale Factor=<0.6 15>
    Min Scale Factor=<0.2 1>
    Sample Scale Factor By Curve=true
    Scale Factor Curve=DataInterface() {propertyValues {Curve {keys [
        <at -0.0077 0.0765 RCIM_Cubic>
        {arriveTangent=1.2463, interpMode=RCIM_Cubic, leaveTangent=1.2463, tangentMode=RCTM_User, time=0.9985, value=0.9549}
       ]}}}
   ) {enabled=false}
   &ScaleColor(
    Scale Alpha=&FloatFromCurve(FloatCurve=DataInterface() {propertyValues {Curve {
        defaultValue=0
        keys [{interpMode=RCIM_Cubic, tangentMode=RCTM_User, time=0, value=1}, {interpMode=RCIM_Cubic, tangentMode=RCTM_User, time=1, value=0}]
       }}})
    ScaleA=true
    ScaleRGB=true
   )
   /SetVariables(
    Particles.Linear Color=&MultiplyLinearColor(Linear Color=$Particles.Color, Scale Factor=<0.1 0.1 0 0.05>)
    Particles.Vector 2D:Vector2f=&Multiply_Vector2D(A=$Particles.SpriteSize, B=<10 2>)
   ) {isSetParameters=true}
  ]
  rendererClasses [&NiagaraSpriteRendererProperties]
  renderers [
   {rendererClass=&NiagaraSpriteRendererProperties}
   {rendererClass=&NiagaraSpriteRendererProperties, rendererIndex=1}
  ]
 }
 snow_mist001 {
  emitterSpawnScript [/SetVariables(Emitter.FE Particle Spawn Rate=35, Emitter.Idle Spawn Rate=60, Emitter.Max Traversal Speed=150, Emitter.Particle Spawn Rate=60) {enabled=false, isSetParameters=true}]
  emitterUpdateScript [
   &EmitterState(Life Cycle Mode=::NewEnumerator1, Loop Behavior=::NewEnumerator1, Loop Duration=0.15, UseLoopDelay=false)
   &SpawnBurst_Instantaneous(Spawn Count=300, Use Spawn Probability=false, UseLoopCountLimit=false)
   &SpawnRate(SpawnRate=60, Use Spawn Probability=false) {enabled=false, module=SpawnRate001}
   &SpawnRate(
    SpawnRate=&FN_FrontEndFloatModifier(
     Gameplay=&MakeCustomFloatFromBool(
      Bool=&SetBoolByFloatComparison(A=&VectorLength(Vector=$User.OwnerLocalVelocity), B=300)
      True Float=$Emitter.Particle Spawn Rate
      False Float=$Emitter.Idle Spawn Rate
     )
     Front End=$Emitter.FE Particle Spawn Rate
    )
    Use Spawn Probability=false
   ) {enabled=false}
   /SetVariables(Emitter.EmitterColor:LinearColor=&ColorFromCurve(
     ColorCurve=DataInterface() {propertyValues {
       AlphaCurve {keys [<at 0 1 RCIM_Cubic>,<at 1 0 RCIM_Cubic>]}
       BlueCurve {keys [<at 0 1.1434>,<at 0.0034 0.2473>,<at 0.1431 1 RCIM_Cubic>]}
       GreenCurve {keys [<at 0 4.8938>,<at 0.0034 0.5973>,<at 0.1431 1 RCIM_Cubic>]}
       RedCurve {keys [<at 0 10>,<at 0.0034 1>,<at 0.1431 1 RCIM_Cubic>]}
      }}
     CurveIndex=&Divide_Float(A=$Emitter.Age, B=2)
    )) {enabled=false, isSetParameters=true}
  ]
  particleSpawnScript [
   &InitializeParticle(
    Color=&MultiplyLinearColor(Linear Color=<rgba 0.9062 0.7526 0.3493 1>, Scale Factor=&MakeVector4(X=1, Y=1, Z=1, W=0.3))
    Lifetime=&Multiply_Float(A=&RandomRangeFloat(Minimum=0.55, Maximum=0.8), B=0.7)
    Sprite Rotation Mode=::NewEnumerator2
    Sprite Size Mode=::NewEnumerator1
    Uniform Sprite Size Max=80
    Uniform Sprite Size Min=80
    UsePositionOffset=false
    Write Material Random=true
   )
   &ShapeLocation(
    Cylinder Height=0
    Cylinder Radius=1
    MaskSpawnGroup=false
    Override Seed=false
    Rotation Axis=<0 1 0>
    Rotation Mode=::NewEnumerator1
    Shape Primitive=::NewEnumerator1
    Surface Only=false
    UseOwnerScale=true
   )
   &TorusLocation(
    Align Torus With Axis=true
    Large Radius:NiagaraFloat=&Multiply_Float(A=30, B=&ClampFloat(Float=$User.ForwardDot, Min=0.3))
    MaskSpawnGroup=false
    Module.Override Randomness=false
    Torus Axis=&ReturnCameraVector(Camera Vectors=::NewEnumerator4, Coordinate Space=::Simulation, Origin Position=$Particles.Position)
    Torus Mode=::NewEnumerator2
   ) {enabled=false}
   &AddVelocityFromPoint(
    Module.Use Falloff Distance=false
    Origin Offset Coordinate Space=::Simulation
    Velocity Strength=&Multiply_Float(
     A=500
     B=&FloatFromCurve(FloatCurve=DataInterface() {propertyValues {Curve {keys [<at 0 0.5>,<at 0.1 1>,<at 1 0>]}}})
    )
   )
   &AddVelocity(Velocity=<0 0 800>) {enabled=false}
   &AddVelocityInCone(Cone Angle=10, Cone Axis=<0 0 1>, Module.Use Velocity Falloff On Cone Axis=false, Velocity Strength=150) {enabled=false, module=AddVelocityInCone001}
   &CameraOffset(Camera Offset Amount=10)
   &DynamicMaterialParameters(
    Index 0 Param 1=1
    Index 0 Param 2=1
    Index 0 Param 3=1
    Index 0 Param 4=1
    Module.Param0WriteEnabled=true
    Module.Param1WriteEnabled=true
    Module.Param2WriteEnabled=true
    Module.Param3WriteEnabled=true
   )
  ]
  particleUpdateScript [
   &ParticleState()
   &AccelerationForce(Acceleration=&MakeVector(X=300, Y=0, Z=300))
   &GravityForce(Gravity=&RandomRangeVector(Minimum=<0 0 -300>, Maximum=<0 0 -500>)) {enabled=false}
   &Drag(Drag=&RandomRangeFloat(Minimum=1, Maximum=10), Use Linear Drag=true, Use Rotational Drag=true)
   &SolveForcesAndVelocity(Clamp Velocity=false, Limit Acceleration=false)
   &ScaleSpriteSize(
    Scale Sprite Size Mode=::NewEnumerator0
    Uniform Scale Factor=&FloatFromCurve(
     FloatCurve=DataInterface() {propertyValues {Curve {keys [
         {arriveTangent=0.8753, interpMode=RCIM_Cubic, leaveTangent=0.8753, tangentMode=RCTM_User, time=0, value=0.7}
         {arriveTangent=0.0238, interpMode=RCIM_Cubic, leaveTangent=0.0238, tangentMode=RCTM_User, time=1, value=1}
        ]}}}
     Scale Curve=0.75
    )
   )
   &ScaleColor(
    Scale Alpha=&Multiply_Float(
     A=&FloatFromCurve(
      FloatCurve=DataInterface() {propertyValues {Curve {keys [
          {interpMode=RCIM_Cubic, tangentMode=RCTM_User, time=0, value=0}
          {interpMode=RCIM_Cubic, leaveTangent=-2.6773, tangentMode=RCTM_Break, time=0.01, value=1}
          {tangentMode=RCTM_User, time=1, value=0}
         ]}}}
      Scale Curve=&RandomRangeFloat(Minimum=0.5, Maximum=1)
     )
     B=&FNTimeOfDayScalar(Night Scale=0.6, Morning Scale=0.7, Evening Scale=0.5)
    )
    ScaleA=true
    ScaleRGB=true
   )
   &SubUVAnimation(SubUV Loop Count=0.3)
   &BlendSpace("Position Blend (World <-> Local)":NiagaraFloat=&MakeCustomFloatFromBool(
     Bool=&SetBoolByFloatComparison(A=&VectorLength(Vector=$Engine.Owner.Velocity), B=400)
     True Float=1
     False Float=&FloatFromCurve(FloatCurve=DataInterface() {propertyValues {Curve {keys [<at 0 1 RCIM_Cubic>, {arriveTangent=-3.5186, interpMode=RCIM_Cubic, leaveTangent=-3.5186, tangentMode=RCTM_User, time=1, value=0}]}}})
     Translation Contribution:NiagaraFloat=1
     Rotation Contribution:NiagaraFloat=1
    )) {enabled=false}
   /SetVariables(Particles.Color=&MultiplyLinearColor(
     Linear Color=$Particles.Color
     Scale Factor=&MakeVector4_FromLinearColor(LinearColor=$Emitter.EmitterColor)
    )) {isSetParameters=true}
  ]
  rendererClasses [&NiagaraSpriteRendererProperties]
  renderers [{rendererClass=&NiagaraSpriteRendererProperties}]
 }
}
summary {
 systemName=NS_Emote_WavySignalBeef_Flash1
 emitters [
  {emitterName=Ray_01, rendererClasses [&NiagaraSpriteRendererProperties]}
  {emitterName=Ray_01001, rendererClasses [&NiagaraSpriteRendererProperties]}
  {emitterName=snow_mist001, rendererClasses [&NiagaraSpriteRendererProperties]}
  {emitterName=Sparks001, rendererClasses [&NiagaraSpriteRendererProperties]}
  {emitterName=FlashElements002, rendererClasses [&NiagaraLightRendererProperties]}
 ]
 userVariables [
  User.bIsFrontEnd {type=NiagaraBool}
  User.bIsFrontEndPreview {type=NiagaraBool}
  User.ForwardDot {type=NiagaraFloat, defaultValue=0, description=""}
  User.FrontEndScale {type=NiagaraFloat, defaultValue=0.6, description=""}
  User.OwnerLocalVelocity {type=Vector3f, description="", defaultValue {struct=Vector3f, value=<0 0 0>}}
  User.SpawnGroup {type=NiagaraInt32}
 ]
}
